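# Landscape material function that recolors grass base color by biome for the Hermes terrain, with a separate mobile result.
# The main path blends four slices of a biome color texture array using masks, then adjusts color by world height relative to ocean level.
# The mobile path lerps tinted colors through biome masks.
# A static switch gates both results; per-biome palette and Bigsplat height-lerp chains end unconnected.
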MaterialFunction /BRCoreMaterials/Hermes/Terrain/Landscape/Materials/Layers/BiomeColor/MF_Hermes_Grass_BiomeColor
graph {
 properties {libraryCategoriesText [FN-Landscape]}
 nodes [
  Constant3Vector() {constant=<rgba 0.0602 0.07 0.0432 1>}
  Constant3Vector() {constant=<rgba 0.1 0.0871 0.0643 1>}
  Constant3Vector() {constant=<rgba 0.115 0.1144 0.065 1>}
  Constant3Vector() {constant=<rgba 0.0938 0.0742 0.0683 1>}
  Constant3Vector() {constant=<rgba 0.0059 0.1 0.0808 1>}
  Constant3Vector() {constant=<rgba 0.007 0.07 0.0699 1>}
  Constant3Vector() {constant=<rgba 0 0.2 0.1893 1>}
  Constant3Vector() {constant=<rgba 0.032 0.16 0.1032 1>}
  Constant() {r=0.14}
  n1 = VectorParameter() {group=BiomeColors, parameterName=Mobile_GrassColor_MaskMaker, defaultValue=<rgba 0.4 0.7208 2 1>}
  Desaturation(None=NamedRerouteUsage(), Fraction=@n1.A)
  FunctionInput() {bUsePreviewValueAsDefault=true, inputName=TerrainMasks_QuadD, inputType=FunctionInput_Vector4, sortPriority=40}
  n2 = NamedRerouteUsage()
  n3 = NamedRerouteUsage()
  FunctionOutput(None=SetMaterialAttributes!(
    MaterialAttributes=NamedRerouteUsage()
    BaseColor=StaticSwitchParameter(True=@n2, False=@n3) {defaultValue=true, parameterName=UseHermesBIomeColors}
   ) {attributeSetTypes [#1]})
  n4 = NamedRerouteUsage()
  n5 = NamedRerouteUsage()
  FunctionOutput(None=SetMaterialAttributes(
    MaterialAttributes=NamedRerouteUsage()
    BaseColor=StaticSwitchParameter(True=@n4, False=@n5) {defaultValue=true, parameterName=UseHermesBIomeColors}
   ) {attributeSetTypes [#1]}) {outputName=Result Mobile, sortPriority=1}
  n6 = VectorParameter() {group=BiomeColors, parameterName=Bigsplat2_C01, defaultValue=<rgba 3 1.7278 1.2 1>}
  n7 = MaterialFunctionCall!(UseLandscapeCoords=StaticBoolParameter() {parameterName=UseLandscapeCoords}, "UseProjection(translucent)") {materialFunction=/BRCoreMaterials/Hermes/Terrain/BiomeMask/MF_Hermes_BiomeMasks}
  n8 = AppendVector(A=@n7.UVs, B=Constant() {r=3})
  n9 = TextureObjectParameter() {group=BiomeColors, parameterName=GrassColor_TextureArray, texture=/BRCoreMaterials/Hermes/Terrain/BiomeMask/T_HermesBiomeGrass_Array}
  n10 = Constant()
  n11 = NamedRerouteDeclaration(None=TextureSample(UVs=AppendVector(A=@n7.UVs, B=@n10), Tex=@n9) {samplerSource=SSM_Wrap_WorldGroupSettings, texture=&T_HermesBiomeGrass_Array}.RGB) {name=BiomeGrassColorTexture01, nodeColor=<rgba 1 0 1 1>}
  n12 = AppendVector(A=@n7.UVs, B=Constant() {r=1})
  n13 = LinearInterpolate(
   A=@n11
   B=NamedRerouteDeclaration(None=TextureSample(UVs=@n12, Tex=@n9) {samplerSource=SSM_Wrap_WorldGroupSettings, texture=&T_HermesBiomeGrass_Array}.RGB) {name=BiomeGrassColorTexture02, nodeColor=<rgba 1 0 1 1>}
   Alpha=NamedRerouteUsage()
  )
  n14 = Desaturation(None=@n11, Fraction=NamedRerouteUsage())
  n15 = NamedRerouteUsage()
  n16 = TextureSample(UVs=AppendVector(A=@n7.UVs, B=Constant() {r=2}), Tex=@n9) {samplerSource=SSM_Wrap_WorldGroupSettings, texture=&T_HermesBiomeGrass_Array}
  n17 = LinearInterpolate(
   A=NamedRerouteDeclaration(None=TextureSample(UVs=@n8, Tex=@n9) {samplerSource=SSM_Wrap_WorldGroupSettings, texture=&T_HermesBiomeGrass_Array}.RGB) {name=BiomeGrassColorTexture04, nodeColor=<rgba 1 0 1 1>}
   B=LinearInterpolate(
    A=LinearInterpolate(A=@n13, B=@n14, Alpha=@n15)
    B=NamedRerouteDeclaration(None=@n16.RGB) {name=BiomeGrassColorTexture03, nodeColor=<rgba 1 0 1 1>}
    Alpha=NamedRerouteUsage()
   )
   Alpha=NamedRerouteUsage()
  )
  n18 = NamedRerouteUsage()
  n19 = ScalarParameter() {defaultValue=0.5, group=BiomeColors, parameterName=K_Bigsplat_NoiseTransPhasePower}
  n20 = Saturate(None=Multiply(A=Power(Base=@n18, Exp=@n19), B=ScalarParameter() {defaultValue=0.6, group=BiomeColors, parameterName=K_Bigsplat_NoiseTransPhaseMult}))
  n21 = NamedRerouteUsage()
  n22 = VectorParameter() {parameterName=K_BigsplatNoiseChannel, defaultValue=<rgba 1 0 0 0>}
  n23 = ScalarParameter() {defaultValue=3100, group=BiomeColors, parameterName=K_Bigsplat_GradientNoiseScale}
  n24 = DotProduct(
   A=MaterialFunctionCall(X=@n22.R, Y=@n22.R, Z=@n22.R, A=@n22.R) {materialFunction=&MakeFloat4}.Result
   B=TextureSample(UVs=MaterialFunctionCall!(Scale=@n23) {materialFunction=/Game/Athena/Apollo/Environments/Landscape/Materials/Biome/CommonMaterials/MF_SimpleWorldUV}.Result) {desc=T_Malachite_Noise, samplerSource=SSM_Wrap_WorldGroupSettings}.RGBA
  )
  LinearInterpolate(
   A=@n17
   B=Saturate(None=Multiply(A=@n17, B=@n6.RGB))
   Alpha=MaterialFunctionCall(
    Transition Phase=Multiply(A=@n20, B=@n21)
    Height Texture=OneMinus(None=@n24)
    Contrast=ScalarParameter() {defaultValue=0.5, group=BiomeColors, parameterName=K_Bigsplat_HeightlerpContrast}
   ) {materialFunction=&HeightLerp}.Alpha
  )
  n25 = NamedRerouteUsage()
  n26 = Saturate(None=Multiply(A=VectorParameter() {group=BiomeColors, parameterName=Mobile_GrassColor_Bigsplat, defaultValue=<rgba 0 0.9107 1 1>}.RGB, B=@n25))
  n27 = VectorParameter() {group=BiomeColors, parameterName=Mobile_GrassColor_Bigsplat2, defaultValue=<rgba 0 1.2072 1.5 1>}
  LinearInterpolate(A=@n26, B=Saturate(None=Multiply(A=@n25, B=@n27.RGB)), Alpha=NamedRerouteUsage())
  n28 = NamedRerouteUsage()
  n29 = VectorParameter() {group=BiomeColors, parameterName=Mobile_GrassColor_Mountains, defaultValue=<rgba 0.1969 0.6302 0.6239 1>}
  n30 = Multiply(A=@n29.RGB, B=NamedRerouteUsage())
  n31 = VectorParameter() {group=BiomeColors, parameterName=Mobile_GrassColor_Wetlands, defaultValue=<rgba 0 0.7569 2.7 1>}
  n32 = NamedRerouteUsage()
  n33 = LinearInterpolate(
   A=LinearInterpolate(A=@n28, B=Saturate(None=@n30), Alpha=NamedRerouteUsage())
   B=Saturate(None=Multiply(A=@n31.RGB, B=@n32))
   Alpha=NamedRerouteUsage()
  )
  n34 = VectorParameter() {group=BiomeColors, parameterName=Mobile_GrassColor_MossyForest, defaultValue=<rgba 0.847 1.1 0.451 1>}
  n35 = NamedRerouteUsage()
  LinearInterpolate(A=@n33, B=Saturate(None=Multiply(A=@n34.RGB, B=@n35)), Alpha=NamedRerouteUsage())
  n36 = VectorParameter() {group=BiomeColors, parameterName=MT_CD, defaultValue=<rgba 0.0253 0.09 0.0345 1>}
  n37 = VectorParameter() {group=BiomeColors, parameterName=MT_CA, defaultValue=<rgba 0.025 0.1 0.0274 1>}
  n38 = VectorParameter() {group=BiomeColors, parameterName=MT_CB, defaultValue=<rgba 0.0452 0.15 0.0497 1>}
  n39 = NamedRerouteUsage()
  n40 = NamedRerouteUsage()
  n41 = LinearInterpolate(
   A=LinearInterpolate(A=LinearInterpolate(A=@n37.RGB, B=@n38.RGB, Alpha=@n39), B=Desaturation(None=@n37.RGB, Fraction=@n40), Alpha=NamedRerouteUsage())
   B=VectorParameter() {group=BiomeColors, parameterName=MT_CC, defaultValue=<rgba 0.0512 0.15 0.0416 1>}.RGB
   Alpha=NamedRerouteUsage()
  )
  n42 = NamedRerouteUsage()
  n43 = VectorParameter() {group=BiomeColors, parameterName=WL_CA, defaultValue=<rgba 0.0273 0.1779 0.0467 1>}
  n44 = LinearInterpolate(
   A=LinearInterpolate(
    A=@n43.RGB
    B=VectorParameter() {group=BiomeColors, parameterName=WL_CB, defaultValue=<rgba 0 0.1137 0.0275 1>}.RGB
    Alpha=NamedRerouteUsage()
   )
   B=Desaturation(None=@n43.RGB, Fraction=NamedRerouteUsage())
   Alpha=NamedRerouteUsage()
  )
  n45 = VectorParameter() {group=BiomeColors, parameterName=WL_CC, defaultValue=<rgba 0.0745 0.1725 0.0235 1>}
  n46 = NamedRerouteUsage()
  n47 = LinearInterpolate(
   A=LinearInterpolate(A=NamedRerouteUsage(), B=LinearInterpolate(A=@n36.RGB, B=@n41, Alpha=@n42), Alpha=NamedRerouteUsage())
   B=LinearInterpolate(
    A=VectorParameter() {group=BiomeColors, parameterName=WL_CD, defaultValue=<rgba 0.0235 0.1294 0.0392 1>}.RGB
    B=LinearInterpolate(A=@n44, B=@n45.RGB, Alpha=@n46)
    Alpha=NamedRerouteUsage()
   )
   Alpha=NamedRerouteUsage()
  )
  n48 = VectorParameter() {group=BiomeColors, parameterName=MF_CA, defaultValue=<rgba 0.0952 0.25 0.0102 1>}
  n49 = LinearInterpolate(
   A=@n48.RGB
   B=VectorParameter() {group=BiomeColors, parameterName=MF_CB, defaultValue=<rgba 0.0585 0.13 0.0076 1>}.RGB
   Alpha=NamedRerouteUsage()
  )
  n50 = Desaturation(None=@n48.RGB, Fraction=NamedRerouteUsage())
  n51 = NamedRerouteUsage()
  n52 = LinearInterpolate(
   A=VectorParameter() {group=BiomeColors, parameterName=MF_CD, defaultValue=<rgba 0.0717 0.15 0.003 1>}.RGB
   B=LinearInterpolate(
    A=LinearInterpolate(A=@n49, B=@n50, Alpha=@n51)
    B=VectorParameter() {group=BiomeColors, parameterName=MF_CC, defaultValue=<rgba 0.1379 0.25 0.0208 1>}.RGB
    Alpha=NamedRerouteUsage()
   )
   Alpha=NamedRerouteUsage()
  )
  n53 = NamedRerouteUsage()
  n54 = VectorParameter() {group=BiomeColors, parameterName=BF_CD, defaultValue=<rgba 0.0954 0.135 0.018 1>}
  n55 = VectorParameter() {group=BiomeColors, parameterName=BF_CA, defaultValue=<rgba 0.18 0.0701 0.009 1>}
  n56 = VectorParameter() {group=BiomeColors, parameterName=BF_CB, defaultValue=<rgba 0.3 0.1112 0.018 1>}
  n57 = NamedRerouteUsage()
  n58 = NamedRerouteUsage()
  n59 = LinearInterpolate(
   A=LinearInterpolate(A=LinearInterpolate(A=@n55.RGB, B=@n56.RGB, Alpha=@n57), B=Desaturation(None=@n55.RGB, Fraction=@n58), Alpha=NamedRerouteUsage())
   B=VectorParameter() {group=BiomeColors, parameterName=BF_CC, defaultValue=<rgba 0.23 0.0655 0.023 1>}.RGB
   Alpha=NamedRerouteUsage()
  )
  n60 = NamedRerouteUsage()
  n61 = Constant3Vector() {constant=<rgba 0.0356 0.0595 0.0194 1>}
  n62 = LinearInterpolate(
   A=LinearInterpolate(A=@n61, B=Constant3Vector() {constant=<rgba 0.0648 0.0931 0.0307 1>}, Alpha=NamedRerouteUsage())
   B=Desaturation(None=@n61, Fraction=NamedRerouteUsage())
   Alpha=NamedRerouteUsage()
  )
  n63 = Constant3Vector() {constant=<rgba 0.0999 0.0953 0.0242 1>}
  n64 = NamedRerouteUsage()
  n65 = LinearInterpolate(
   A=LinearInterpolate(A=LinearInterpolate(A=@n47, B=@n52, Alpha=@n53), B=LinearInterpolate(A=@n54.RGB, B=@n59, Alpha=@n60), Alpha=NamedRerouteUsage())
   B=LinearInterpolate(A=Constant3Vector() {constant=<rgba 0.0782 0.0887 0.0252 1>}, B=LinearInterpolate(A=@n62, B=@n63, Alpha=@n64), Alpha=NamedRerouteUsage())
   Alpha=NamedRerouteUsage()
  )
  n66 = VectorParameter() {group=BiomeColors, parameterName=RF_CA, defaultValue=<rgba 0.2937 0.0194 0.0463 1>}
  n67 = LinearInterpolate(
   A=@n66.RGB
   B=VectorParameter() {group=BiomeColors, parameterName=RF_CB, defaultValue=<rgba 0.069 0.08 0.0125 1>}.RGB
   Alpha=NamedRerouteUsage()
  )
  n68 = Desaturation(None=@n66.RGB, Fraction=NamedRerouteUsage())
  n69 = NamedRerouteUsage()
  n70 = LinearInterpolate(
   A=VectorParameter() {group=BiomeColors, parameterName=RF_CD, defaultValue=<rgba 0.1604 0.0151 0.0234 1>}.RGB
   B=LinearInterpolate(
    A=LinearInterpolate(A=@n67, B=@n68, Alpha=@n69)
    B=VectorParameter() {group=BiomeColors, parameterName=RF_CC, defaultValue=<rgba 0.0842 0.15 0.0125 1>}.RGB
    Alpha=NamedRerouteUsage()
   )
   Alpha=NamedRerouteUsage()
  )
  n71 = NamedRerouteUsage()
  n72 = VectorParameter() {group=BiomeColors, parameterName=MM_C04, defaultValue=<rgba 0.0359 0.099 0.083 1>}
  n73 = VectorParameter() {group=BiomeColors, parameterName=MM_C01, defaultValue=<rgba 0.0262 0.0762 0.0762 1>}
  n74 = VectorParameter() {group=BiomeColors, parameterName=MM_C02, defaultValue=<rgba 0.0396 0.1406 0.1122 1>}
  n75 = NamedRerouteUsage()
  n76 = NamedRerouteUsage()
  n77 = LinearInterpolate(
   A=LinearInterpolate(A=LinearInterpolate(A=@n73.RGB, B=@n74.RGB, Alpha=@n75), B=Desaturation(None=@n73.RGB, Fraction=@n76), Alpha=NamedRerouteUsage())
   B=VectorParameter() {group=BiomeColors, parameterName=MM_C03, defaultValue=<rgba 0.053 0.0613 0.2292 1>}.RGB
   Alpha=NamedRerouteUsage()
  )
  n78 = NamedRerouteUsage()
  n79 = VectorParameter() {group=BiomeColors, parameterName=Bigsplat_C01, defaultValue=<rgba 0.0047 0.11 0.0596 1>}
  n80 = LinearInterpolate(
   A=LinearInterpolate(
    A=@n79.RGB
    B=VectorParameter() {group=BiomeColors, parameterName=Bigsplat_C02, defaultValue=<rgba 0.0037 0.125 0.0332 1>}.RGB
    Alpha=NamedRerouteUsage()
   )
   B=Desaturation(None=@n79.RGB, Fraction=NamedRerouteUsage())
   Alpha=NamedRerouteUsage()
  )
  n81 = VectorParameter() {group=BiomeColors, parameterName=Bigsplat_C03, defaultValue=<rgba 0.004 0.2 0.1064 1>}
  n82 = NamedRerouteUsage()
  LinearInterpolate(
   A=LinearInterpolate(A=LinearInterpolate(A=@n65, B=@n70, Alpha=@n71), B=LinearInterpolate(A=@n72.RGB, B=@n77, Alpha=@n78), Alpha=NamedRerouteUsage())
   B=LinearInterpolate(
    A=VectorParameter() {group=BiomeColors, parameterName=Bigsplat_C04, defaultValue=<rgba 0.0336 0.16 0.0674 1>}.RGB
    B=LinearInterpolate(A=@n80, B=@n81.RGB, Alpha=@n82)
    Alpha=NamedRerouteUsage()
   )
   Alpha=NamedRerouteUsage()
  )
  n83 = VectorParameter() {group=BiomeColors, parameterName=TEST_C04, defaultValue=<rgba 0.032 0.16 0.1032 1>}
  n84 = VectorParameter() {group=BiomeColors, parameterName=TEST_C01, defaultValue=<rgba 0.0059 0.1 0.0808 1>}
  n85 = VectorParameter() {group=BiomeColors, parameterName=TEST_C02, defaultValue=<rgba 0.007 0.07 0.0699 1>}
  n86 = NamedRerouteUsage()
  n87 = NamedRerouteUsage()
  n88 = LinearInterpolate(
   A=LinearInterpolate(A=LinearInterpolate(A=@n84.RGB, B=@n85.RGB, Alpha=@n86), B=Desaturation(None=@n84.RGB, Fraction=@n87), Alpha=NamedRerouteUsage())
   B=VectorParameter() {group=BiomeColors, parameterName=TEST_C03, defaultValue=<rgba 0 0.2 0.1893 1>}.RGB
   Alpha=NamedRerouteUsage()
  )
  n89 = NamedRerouteUsage()
  LinearInterpolate(
   A=@n17
   B=LinearInterpolate(
    A=LinearInterpolate(A=@n83.RGB, B=@n88, Alpha=@n89)
    B=VectorParameter() {group=BiomeColors, parameterName=DEBUG_TempBiomeGrass_COLOR, defaultValue=<rgba 0.7188 0.0061 0 1>}.RGB
    Alpha=ScalarParameter() {group=BiomeColors, parameterName=DEBUG_TempBiomeGrassColorLerp}
   )
   Alpha=NamedRerouteUsage()
  )
  n90 = NamedRerouteUsage()
  n91 = Multiply(A=ComponentMask(None=NamedRerouteUsage()) {g=true}) {constB=0.4}
  Multiply(A=OneMinus(None=@n90), B=@n91)
  Multiply(A=@n28, B=VectorParameter() {group=BiomeColors, parameterName=Mobile_GrassColor_Grasslands, defaultValue=<rgba 0.3364 1 0.1156 1>}.RGB)
  n92 = MaterialFunctionCall(Float4=FunctionInput() {bUsePreviewValueAsDefault=true, inputName=TerrainMasks_QuadA, inputType=FunctionInput_Vector4, sortPriority=10}) {materialFunction=&BreakOutFloat4Components}
  NamedRerouteDeclaration(None=@n92.R) {name=BM_Mountains, nodeColor=<rgba 0.5 0 1 1>}
  NamedRerouteDeclaration(None=@n92.G) {name=BM_Wetlands, nodeColor=<rgba 0.5 0 1 1>}
  n93 = Constant() {r=0.7}
  n94 = TruncateLWC!(None=WorldPosition().Z)
  n95 = Divide(
   A=Add(
    A=@n94
    B=Constant() {desc="This offsets the ocean level used by the sand. Positive numbers go lower, negative numbers make it go higher.", r=-200}
   )
   B=Constant() {r=1000}
  )
  n96 = Saturate(None=Add(A=Saturate(None=@n95), B=Constant() {r=0.58}))
  n97 = OneMinus(None=NamedRerouteUsage())
  n98 = Multiply(A=Saturate(None=Add(A=@n96, B=@n97)), B=@n17)
  n99 = MaterialFunctionCall(Float3=@n98) {materialFunction=&BreakOutFloat3Components}
  n100 = OneMinus(None=@n99.R)
  n101 = Power(
   Base=OneMinus(None=MaterialFunctionCall(X=@n100, Y=@n99.G, Z=@n99.G) {materialFunction=&MakeFloat3}.Result)
   Exp=Constant() {r=20}
  ) {constExponent=16}
  n102 = Constant4Vector() {constant=<rgba 0.13 0.1283 0.013 0.5>}
  n103 = Multiply(
   A=Desaturation(
    None=MaterialFunctionCall(Hue Shift Percentage=@n93, Texture=@n101) {materialFunction=&HueShift}.Result
    Fraction=@n102.A
   )
   B=ComponentMask(None=@n102.RGBA) {b=true, g=true, r=true}
  )
  n104 = Add(
   A=@n94
   B=Constant() {desc="This offsets the ocean level used by the sand. Positive numbers go lower, negative numbers make it go higher.", r=-2200}
  )
  n105 = Constant() {r=2500}
  n106 = Multiply(A=Saturate(None=Divide(A=@n104, B=@n105)), B=NamedRerouteUsage())
  NamedRerouteDeclaration(None=LinearInterpolate(A=Saturate(None=@n103), B=@n98, Alpha=OneMinus(None=@n106))) {name=OutputBasecolor, nodeColor=<rgba 0.5 0 1 1>}
  n107 = VectorParameter() {group=BiomeColors, parameterName=GL_CA, defaultValue=<rgba 0.0845 0.1732 0.0065 1>}
  n108 = ScalarParameter() {defaultValue=0.14, group=Grass, parameterName=Noise02_Desaturation}
  n109 = LinearInterpolate(
   A=LinearInterpolate(
    A=@n107.RGB
    B=VectorParameter() {group=BiomeColors, parameterName=GL_CB, defaultValue=<rgba 0.0651 0.1137 0.0105 1>}.RGB
    Alpha=NamedRerouteUsage()
   )
   B=Desaturation(None=@n107.RGB, Fraction=NamedRerouteDeclaration(None=@n108) {name=Noise02_Desat, nodeColor=<rgba 0.5 0 1 1>})
   Alpha=NamedRerouteUsage()
  )
  n110 = VectorParameter() {group=BiomeColors, parameterName=GL_CC, defaultValue=<rgba 0.1263 0.144 0.0199 1>}
  n111 = NamedRerouteUsage()
  NamedRerouteDeclaration(None=LinearInterpolate(
    A=VectorParameter() {group=BiomeColors, parameterName=GL_CD, defaultValue=<rgba 0.0503 0.13 0.0125 1>}.RGB
    B=LinearInterpolate(A=@n109, B=@n110.RGB, Alpha=@n111)
    Alpha=NamedRerouteUsage()
   )) {name=NoBiomeColor, nodeColor=<rgba 0.5 0 1 1>}
  n112 = FunctionInput() {bUsePreviewValueAsDefault=true, inputName=GrassNoise, inputType=FunctionInput_Vector4, sortPriority=50, previewValue=<0 0 0 0>}
  NamedRerouteDeclaration(None=@n112) {name=GrassNoiseVec4, nodeColor=<rgba 0.5 0 1 1>}
  n113 = MaterialFunctionCall(Float4=FunctionInput() {bUsePreviewValueAsDefault=true, inputName=TerrainMasks_QuadB, inputType=FunctionInput_Vector4, sortPriority=20}) {materialFunction=&BreakOutFloat4Components}
  NamedRerouteDeclaration(None=@n113.R) {name=BM_Goldlands, nodeColor=<rgba 0.5 0 1 1>}
  n114 = GetMaterialAttributes(None=FunctionInput(Preview=SetMaterialAttributes!(MaterialAttributes)) {bUsePreviewValueAsDefault=true, inputName=Material, inputType=FunctionInput_MaterialAttributes}) {attributeGetTypes [#1, #2]}
  n115 = NamedRerouteDeclaration(None=@n114.BaseColor) {name=Input_BaseColor, nodeColor=<rgba 0.5 0 1 1>}
  NamedRerouteDeclaration(None=Desaturation(None=@n115, Fraction=Constant() {r=0.5})) {name=BaseColor_DesatHalf, nodeColor=<rgba 0.5 0 1 1>}
  NamedRerouteDeclaration(None=@n113.G) {name=BM_RedForest, nodeColor=<rgba 0.5 0 1 1>}
  n116 = Constant()
  n117 = Constant() {r=1.5}
  n118 = MaterialFunctionCall!(Coordinates, R_01_Scale, R_02_Scale, R_03_Scale, R_04_Scale, G_01_Scale, G_02_Scale, G_03_Scale, G_04_Scale) {materialFunction=/Game/Athena/Helios/Environments/Landscape/Materials/MaterialFunctions_Utilities/MF_VariationMasks2}
  NamedRerouteDeclaration(None=Saturate(None=LinearInterpolate(A=@n116, B=@n117, Alpha=@n118.G_01))) {name=Noise01, nodeColor=<rgba 0.5 0 1 1>}
  n119 = Constant() {r=1.5}
  n120 = Constant() {r=-9}
  NamedRerouteDeclaration(None=Saturate(None=LinearInterpolate(A=@n119, B=@n120, Alpha=@n118.G_02))) {name=Noise02, nodeColor=<rgba 0.5 0 1 1>}
  n121 = Constant() {r=-0.25}
  n122 = Constant() {r=1.5}
  n123 = NamedRerouteDeclaration(None=Saturate(None=LinearInterpolate(A=@n121, B=@n122, Alpha=@n118.G_03))) {name=Noise03_PreMult, nodeColor=<rgba 0.5 0 1 1>}
  n124 = Constant() {r=0.65}
  NamedRerouteDeclaration(None=Saturate(None=Multiply(A=@n123, B=@n124))) {name=Noise03, nodeColor=<rgba 0.5 0 1 1>}
  NamedRerouteDeclaration(None=@n92.B) {name=BM_MossyForest, nodeColor=<rgba 0.5 0 1 1>}
  n125 = Constant() {r=4}
  NamedRerouteDeclaration(None=Saturate(None=Multiply(A=@n118.G_04, B=@n125))) {name=Noise04, nodeColor=<rgba 0.5 0 1 1>}
  NamedRerouteDeclaration(None=FunctionInput() {bUsePreviewValueAsDefault=true, inputName=VariationMasks Mobile, sortPriority=51}) {name=VariationMasks Mobile, nodeColor=<rgba 1 0 0.6328 1>}
  n126 = MaterialFunctionCall(Float4=FunctionInput() {bUsePreviewValueAsDefault=true, inputName=TerrainMasks_QuadC, inputType=FunctionInput_Vector4, sortPriority=30}) {materialFunction=&BreakOutFloat4Components}
  NamedRerouteDeclaration(None=@n126.G) {name=BM_MaskMaker, nodeColor=<rgba 0.5 0 1 1>}
  NamedRerouteDeclaration(None=@n126.R) {name=BM_Jethro, nodeColor=<rgba 0.5 0 1 1>}
  NamedRerouteDeclaration(None=@n126.B) {name=BM_Bigsplat, nodeColor=<rgba 0.5 0 1 1>}
  NamedRerouteDeclaration(None=@n126.A) {name=BM_Bigsplat2, nodeColor=<rgba 0.5 0 1 1>}
  NamedRerouteDeclaration(None=@n92.A) {name=BM_BorealFall, nodeColor=<rgba 0.5 0 1 1>}
  NamedRerouteDeclaration(None=FunctionInput(Preview=Constant()) {bUsePreviewValueAsDefault=true, inputName=GrassEdgeMask, inputType=FunctionInput_Scalar, sortPriority=1}) {name=GrassEdgeMask, nodeColor=<rgba 0.5 0 1 1>}
  NamedRerouteDeclaration(None=Desaturation(None=@n115, Fraction=Constant() {r=1})) {name=BaseColorDesat, nodeColor=<rgba 0.5 0 1 1>}
  NamedRerouteDeclaration(None=@n114.MaterialAttributes) {name=InputMaterial, nodeColor=<rgba 0.5 0 1 1>}
  NamedRerouteDeclaration(None=@n114.OpacityMask) {name=Heightmap, nodeColor=<rgba 0.5 0 1 1>}
  n127 = Multiply(
   A=VectorParameter() {group=BiomeColors, parameterName=Mobile_GrassColor_BorealFall, defaultValue=<rgba 2.7 1 1 1>}.RGB
   B=NamedRerouteUsage()
  )
  n128 = VectorParameter() {group=BiomeColors, parameterName=Mobile_GrassColor_Goldlands, defaultValue=<rgba 0.6695 0.7083 0.5067 1>}
  n129 = NamedRerouteUsage()
  n130 = LinearInterpolate(
   A=LinearInterpolate(A=@n33, B=Saturate(None=@n127), Alpha=NamedRerouteUsage())
   B=Saturate(None=Multiply(A=@n128.RGB, B=@n129))
   Alpha=NamedRerouteUsage()
  )
  n131 = Saturate(None=Multiply(
    A=VectorParameter() {group=BiomeColors, parameterName=Mobile_GrassColor_RedForest, defaultValue=<rgba 2.5 0.25 0.3465 1>}.RGB
    B=NamedRerouteUsage()
   ))
  n132 = NamedRerouteUsage()
  n133 = Multiply(A=@n1.RGB, B=NamedRerouteUsage())
  n134 = LinearInterpolate(
   A=LinearInterpolate(A=LinearInterpolate(A=@n130, B=@n131, Alpha=@n132), B=Saturate(None=@n133), Alpha=@n90)
   B=@n26
   Alpha=NamedRerouteUsage()
  )
  n135 = NamedRerouteUsage()
  NamedRerouteDeclaration(None=ComponentMask(None=LinearInterpolate(A=@n134, B=@n135, Alpha=@n91)) {b=true, g=true, r=true}) {name=OutputBasecolorMobile, nodeColor=<rgba 0.5 0 1 1>}
  StaticSwitchParameter(False=@n17) {desc="DO NOT ship with this as true!", group=BiomeColors, parameterName=DebugLerpGrassColors}
 ]
}
