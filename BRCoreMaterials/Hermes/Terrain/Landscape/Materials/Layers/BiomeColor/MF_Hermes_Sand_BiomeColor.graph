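# Landscape layer function that tints Hermes sand per biome, overriding base color and roughness behind the UseHermesBIomeColors static switch.
# Chained lerps blend saturated SandColor multiplies for biomes like Mountains, Goldlands, MaskMaker and Bigsplat, using terrain-mask inputs.
# Bigsplat mixes two tints by a squared value, which also modulates roughness.
# A separate mobile output sets only base color; several other biome blends, including Jethro and Wetlands, end unconnected.
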MaterialFunction /BRCoreMaterials/Hermes/Terrain/Landscape/Materials/Layers/BiomeColor/MF_Hermes_Sand_BiomeColor
graph {
 properties {libraryCategoriesText [FN-Landscape]}
 nodes [
  Constant3Vector() {constant=<rgba 0.045 0.2233 0.3 1>}
  Constant3Vector() {constant=<rgba 0.25 0.3952 0.5 1>}
  FunctionInput() {bUsePreviewValueAsDefault=true, inputName=TerrainMasks_QuadD, inputType=FunctionInput_Vector4, sortPriority=40}
  n1 = NamedRerouteUsage()
  n2 = NamedRerouteUsage()
  n3 = NamedRerouteUsage()
  n4 = NamedRerouteUsage()
  FunctionOutput(None=SetMaterialAttributes!(
    MaterialAttributes=NamedRerouteUsage()
    BaseColor=StaticSwitchParameter(True=@n1, False=@n2) {defaultValue=true, parameterName=UseHermesBIomeColors}
    Roughness=StaticSwitchParameter(True=@n3, False=@n4) {defaultValue=true, parameterName=UseHermesBIomeColors}
   ) {attributeSetTypes [#1, #2]})
  FunctionOutput(None=SetMaterialAttributes!(MaterialAttributes=NamedRerouteUsage(), BaseColor=NamedRerouteUsage()) {attributeSetTypes [#1]}) {outputName=Result Mobile, sortPriority=1}
  n5 = Multiply(
   A=VectorParameter() {group=BiomeColors, parameterName=SandColor_Mountains, defaultValue=<rgba 0.2074 0.3094 0.3698 1>}.RGB
   B=NamedRerouteUsage()
  )
  n6 = VectorParameter() {group=BiomeColors, parameterName=SandColor_Goldlands, defaultValue=<rgba 1 1 1 1>}
  n7 = NamedRerouteUsage()
  n8 = LinearInterpolate(
   A=LinearInterpolate(A=NamedRerouteUsage(), B=Saturate(None=@n5), Alpha=NamedRerouteUsage())
   B=Saturate(None=Multiply(A=@n6.RGB, B=@n7))
   Alpha=NamedRerouteUsage()
  )
  n9 = VectorParameter() {group=BiomeColors, parameterName=SandColor_Jethro, defaultValue=<rgba 0.5 0.5 0.5 1>}
  n10 = NamedRerouteUsage()
  LinearInterpolate(A=@n8, B=Saturate(None=Multiply(A=@n9.RGB, B=@n10)), Alpha=NamedRerouteUsage())
  n11 = MaterialFunctionCall(Float4=FunctionInput() {bUsePreviewValueAsDefault=true, inputName=TerrainMasks_QuadA, inputType=FunctionInput_Vector4, sortPriority=10}) {materialFunction=&BreakOutFloat4Components}
  NamedRerouteDeclaration(None=@n11.R) {name=BM_Mountains, nodeColor=<rgba 0.5 0 1 1>}
  NamedRerouteDeclaration(None=@n11.G) {name=BM_Wetlands, nodeColor=<rgba 0.5 0 1 1>}
  n12 = VectorParameter() {group=BiomeColors, parameterName=SandColor_MaskMaker, defaultValue=<rgba 0.175 0.2401 0.35 1>}
  n13 = NamedRerouteUsage()
  n14 = LinearInterpolate(A=@n8, B=Saturate(None=Multiply(A=@n12.RGB, B=@n13)), Alpha=NamedRerouteUsage())
  n15 = VectorParameter() {group=BiomeColors, parameterName=SandColor_Bigsplat, defaultValue=<rgba 0.15 0.2549 0.3 1>}
  n16 = NamedRerouteUsage()
  n17 = VectorParameter() {group=BiomeColors, parameterName=SandColor_Bigsplat2, defaultValue=<rgba 0.21 0.2623 0.3 1>}
  n18 = ScalarParameter() {defaultValue=1, group=Sand, parameterName=Sand_Bigsplat_HeightMult}
  n19 = NamedRerouteUsage()
  n20 = Multiply(A=@n19, B=@n19)
  n21 = LinearInterpolate(A=Saturate(None=Multiply(A=@n15.RGB, B=@n16)), B=Saturate(None=Multiply(A=@n17.RGB, B=@n16)), Alpha=Saturate(None=Multiply(A=@n18, B=@n20)))
  n22 = NamedRerouteUsage()
  NamedRerouteDeclaration(None=ComponentMask(None=LinearInterpolate(A=@n14, B=@n21, Alpha=@n22)) {b=true, g=true, r=true}) {name=OutputBasecolor, nodeColor=<rgba 0.5 0 1 1>}
  n23 = FunctionInput() {bUsePreviewValueAsDefault=true, inputName=GrassNoise, inputType=FunctionInput_Vector4, sortPriority=50, previewValue=<0 0 0 0>}
  NamedRerouteDeclaration(None=@n23) {name=GrassNoiseVec4, nodeColor=<rgba 0.5 0 1 1>}
  n24 = MaterialFunctionCall(Float4=FunctionInput() {bUsePreviewValueAsDefault=true, inputName=TerrainMasks_QuadB, inputType=FunctionInput_Vector4, sortPriority=20}) {materialFunction=&BreakOutFloat4Components}
  NamedRerouteDeclaration(None=@n24.R) {name=BM_Goldlands, nodeColor=<rgba 0.5 0 1 1>}
  n25 = GetMaterialAttributes(None=FunctionInput(Preview=SetMaterialAttributes!(MaterialAttributes)) {bUsePreviewValueAsDefault=true, inputName=Material, inputType=FunctionInput_MaterialAttributes}) {attributeGetTypes [#1, #3, #2]}
  n26 = NamedRerouteDeclaration(None=@n25.BaseColor) {name=Input_BaseColor, nodeColor=<rgba 0.5 0 1 1>}
  NamedRerouteDeclaration(None=Desaturation(None=@n26, Fraction=Constant() {r=0.5})) {name=BaseColor_DesatHalf, nodeColor=<rgba 0.5 0 1 1>}
  NamedRerouteDeclaration(None=@n24.G) {name=BM_RedForest, nodeColor=<rgba 0.5 0 1 1>}
  n27 = Constant()
  n28 = Constant() {r=1.5}
  n29 = MaterialFunctionCall!(Coordinates, R_01_Scale, R_02_Scale, R_03_Scale, R_04_Scale, G_01_Scale, G_02_Scale, G_03_Scale, G_04_Scale) {materialFunction=/Game/Athena/Helios/Environments/Landscape/Materials/MaterialFunctions_Utilities/MF_VariationMasks2}
  NamedRerouteDeclaration(None=Saturate(None=LinearInterpolate(A=@n27, B=@n28, Alpha=@n29.G_01))) {name=Noise01, nodeColor=<rgba 0.5 0 1 1>}
  n30 = Constant() {r=1.5}
  n31 = Constant() {r=-9}
  NamedRerouteDeclaration(None=Saturate(None=LinearInterpolate(A=@n30, B=@n31, Alpha=@n29.G_02))) {name=Noise02, nodeColor=<rgba 0.5 0 1 1>}
  n32 = Constant() {r=-0.25}
  n33 = Constant() {r=1.5}
  n34 = NamedRerouteDeclaration(None=Saturate(None=LinearInterpolate(A=@n32, B=@n33, Alpha=@n29.G_03))) {name=Noise03_PreMult, nodeColor=<rgba 0.5 0 1 1>}
  n35 = Constant() {r=0.65}
  NamedRerouteDeclaration(None=Saturate(None=Multiply(A=@n34, B=@n35))) {name=Noise03, nodeColor=<rgba 0.5 0 1 1>}
  NamedRerouteDeclaration(None=@n11.B) {name=BM_MossyForest, nodeColor=<rgba 0.5 0 1 1>}
  n36 = Constant() {r=4}
  NamedRerouteDeclaration(None=Saturate(None=Multiply(A=@n29.G_04, B=@n36))) {name=Noise04, nodeColor=<rgba 0.5 0 1 1>}
  n37 = MaterialFunctionCall(Float4=FunctionInput() {bUsePreviewValueAsDefault=true, inputName=TerrainMasks_QuadC, inputType=FunctionInput_Vector4, sortPriority=30}) {materialFunction=&BreakOutFloat4Components}
  NamedRerouteDeclaration(None=@n37.G) {name=BM_MaskMaker, nodeColor=<rgba 0.5 0 1 1>}
  NamedRerouteDeclaration(None=@n25.Roughness) {name=Input_Roughness, nodeColor=<rgba 0.5 0 1 1>}
  n38 = ScalarParameter() {group=Sand, parameterName=Sand_Bigsplat_RoughMult}
  n39 = NamedRerouteUsage()
  n40 = OneMinus(None=NamedRerouteUsage())
  n41 = Saturate(None=Multiply(A=ScalarParameter() {defaultValue=4, group=Sand, parameterName=Sand_Bigsplat_HeightMultRough}, B=@n20))
  NamedRerouteDeclaration(None=LinearInterpolate(A=Multiply(A=@n38, B=@n39), B=@n39, Alpha=Max(A=@n40, B=@n41))) {name=OutputRoughness, nodeColor=<rgba 0.5 0 1 1>}
  NamedRerouteDeclaration(None=@n37.B) {name=BM_Bigsplat, nodeColor=<rgba 0.5 0 1 1>}
  NamedRerouteDeclaration(None=@n11.A) {name=BM_BorealFall, nodeColor=<rgba 0.5 0 1 1>}
  NamedRerouteDeclaration(None=@n37.R) {name=BM_Jethro, nodeColor=<rgba 0.5 0 1 1>}
  NamedRerouteDeclaration(None=Desaturation(None=@n26, Fraction=Constant() {r=1})) {name=BaseColorDesat, nodeColor=<rgba 0.5 0 1 1>}
  NamedRerouteDeclaration(None=@n25.MaterialAttributes) {name=InputMaterial, nodeColor=<rgba 0.5 0 1 1>}
  NamedRerouteDeclaration(None=@n25.OpacityMask) {name=Heightmap, nodeColor=<rgba 0.5 0 1 1>}
  NamedRerouteDeclaration(None=ComponentMask(None=Constant3Vector() {constant=<rgba 1 0 0 0>}) {b=true, g=true, r=true}) {name=OutputBasecolorMobile, nodeColor=<rgba 0.5 0 1 1>}
  Power(Base=@n19, Exp=ScalarParameter() {defaultValue=2, group=Sand, parameterName=Sand_Bigsplat_HeightPower})
  n42 = VectorParameter() {group=BiomeColors, parameterName=SandColor_Grasslands, defaultValue=<rgba 1 1 1 1>}
  n43 = NamedRerouteUsage()
  n44 = NamedRerouteDeclaration(None=Saturate(None=Multiply(A=@n42.RGB, B=@n43))) {name=NoBiomeColor, nodeColor=<rgba 0.5 0 1 1>}
  Reroute(None=@n44)
  n45 = Multiply(A=VectorParameter() {group=BiomeColors, parameterName=SandColor_Wetlands, defaultValue=<rgba 1 1 1 1>}.RGB, B=NamedRerouteUsage())
  n46 = VectorParameter() {group=BiomeColors, parameterName=SandColor_MossyForest, defaultValue=<rgba 1 1 1 1>}
  n47 = NamedRerouteUsage()
  n48 = LinearInterpolate(
   A=LinearInterpolate(A=@n44, B=Saturate(None=@n45), Alpha=NamedRerouteUsage())
   B=Saturate(None=Multiply(A=@n46.RGB, B=@n47))
   Alpha=NamedRerouteUsage()
  )
  n49 = Saturate(None=Multiply(A=VectorParameter() {group=BiomeColors, parameterName=SandColor_BorealFall, defaultValue=<rgba 1 1 1 1>}.RGB, B=NamedRerouteUsage()))
  n50 = NamedRerouteUsage()
  n51 = Multiply(A=VectorParameter() {group=BiomeColors, parameterName=SandColor_RedForest, defaultValue=<rgba 1 1 1 1>}.RGB, B=NamedRerouteUsage())
  Reroute(None=LinearInterpolate(A=LinearInterpolate(A=@n48, B=@n49, Alpha=@n50), B=Saturate(None=@n51), Alpha=NamedRerouteUsage()))
 ]
}
